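# Niagara idle effect for the BananaPhilosopher pickaxe that emits twinkling sprite sparkles across the weapon.
# A single Sparkles emitter spawns probabilistically, picking each position from a baked array of local points and each size from a short list.
# Sprites take the system sparkle color, swell and shrink over life on a curve, and spin up.
# Time-of-day scalars scale their RGB and alpha, with reduced values for night, morning and evening.
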NiagaraSystem /BRCosmetics/Weapons/FORT_Melee/Pickaxe_BananaPhilosopher/FX/NS_Pickaxe_BananaPhilosopher_Idle
dependencies {
 usedDataInterfaces [&NiagaraDataInterfaceArrayFloat3, &NiagaraDataInterfaceArrayFloat, &NiagaraDataInterfaceCurve]
 usedDynamicInputs [
  &RandomRangeFloat
  &MakeVector2D
  &FNTimeOfDayColor
 ]
 usedModules [
  &self:SystemSpawnScript.SystemScriptSource.SystemScriptGraph.NiagaraNodeAssignment_0.SetVariables_#1
  &SystemState
 ]
}
emitters {Sparkles {
  emitterSpawnScript [/SetVariables(Emitter.Position Array=DataInterface() {propertyValues {
      GpuSyncMode=SyncCpuToGpu
      MaxElements=0
      FloatData [
       <-12.4259 -0.0112 -7.248>
       <-19.7683 0.3755 -7.6611>
       <-19.8883 0.9372 -7.8493>
       <-15.6909 0.2286 -4.4312>
       <-16.1917 0.2974 5.406>
       <-15.7663 0.3286 0.7802>
       <-3.5681 -1.5506 -10.5854>
       <-9.9729 -2.2036 -10.1242>
       <-11.8254 0.625 -6.7141>
       <-11.9748 2.2222 -6.817>
       <-9.6486 0.2628 -5.6587>
       <-8.2738 1.0048 -5.6261>
       <-4.796 0.2272 -6.7515>
       <-12.6815 0.6906 -8.5695>
       <-15.2386 2.5859 -8.6879>
       <-14.6808 -1.6899 -8.6663>
       <-16.2408 2.6668 -8.7055>
       <-18.1548 -1.0114 -8.6311>
       <-26.1543 -2.1144 -14.0141>
       <-25.2677 -0.0333 -14.1924>
       <-23.8339 -0.3613 -14.4563>
       <-21.2635 -2.0905 -14.8219>
       <-20.3932 1.2264 -14.917>
       <-18.0331 3.0854 -15.0572>
       <-19.2175 0.9762 -15.0241>
       <-16.2618 1.6231 -15.1066>
       <-14.5411 0.2412 -15.0751>
       <-15.0238 3.0819 -15.0886>
       <-14.2094 1.054 -15.0659>
       <-10.3343 1.6805 -14.7855>
       <-11.8424 -1.8362 -14.9503>
       <-9.163 -1.9423 -14.6525>
       <-8.6939 2.7548 -14.5662>
       <-4.5299 -2.3508 -13.7126>
       <-24.8214 -1.2175 -6.5785>
       <-23.5338 -0.8341 -5.5253>
       <-21.4191 -1.4147 -5.9049>
       <-20.0877 -0.4609 -6.7263>
       <-7.129 -0.4239 11.399>
       <-6.178 0.5482 23.8573>
       <-3.3775 -1.0736 25.7745>
       <-7.0536 -0.161 23.2579>
       <-7.8542 -0.7005 22.7098>
       <-10.4652 -0.2697 18.5879>
       <-27.824 -1.4505 17.4355>
       <-28.2417 0.7747 16.5156>
       <-27.8989 1.1171 17.2706>
       <-27.64 -0.0463 20.2962>
       <-29.3634 0.4084 22.1704>
       <-30.399 1.0262 21.0667>
       <-29.4528 1.0088 21.1024>
       <-29.3445 -0.9025 21.2625>
       <-27.9614 -1.2143 20.0781>
       <-28.262 1.3985 16.6567>
       <-17.9715 -0.1035 11.2688>
       <-18.2068 0.3547 8.052>
       <-11.7541 -0.1046 12.266>
       <-11.6899 -0.0642 3.3>
       <-11.5432 0.1249 7.3274>
       <-11.7714 0.4158 -1.166>
       <-19.9122 0.332 13.248>
       <-20.1072 0.4427 -1.7168>
       <-20.2697 0.3943 5.7987>
       <-20.3154 0.3772 12.094>
       <-20.2103 -0.0986 3.9349>
       <-20.1881 -0.0947 -7.044>
       <-20.0012 -0.0482 -1.619>
       <-19.8626 0.2154 1.569>
       <-7.2497 -2.2725 -12.3731>
       <-21.4027 2.1131 -10.5019>
       <-19.7221 2.0223 -11.2036>
       <-8.7493 2.2554 17.5386>
       <-10.175 1.0254 17.6604>
       <-11.9603 0.906 18.1479>
       <-15.5291 2.0567 19.5402>
       <-14.5494 0.3974 19.1276>
       <-16.137 -1.5605 19.6517>
       <-18.6449 -0.9203 18.5954>
       <-20.0516 1.2166 18.1116>
       <-23.5813 2.2129 17.5517>
       <-5.227 1.5252 10.5826>
       <-3.5492 0.2031 11.4519>
       <-1.9415 0.2869 12.2819>
       <-4.3799 1.6707 18.0369>
       <-4.0039 0.2234 17.209>
       <-3.9824 0.2297 20.6315>
       <-3.3994 -1.3414 21.2656>
       <-1.967 -1.6154 22.2112>
       <-1.5332 0.1019 21.1391>
       <-2.2269 -0.1428 20.542>
       <-4.3422 -0.7718 19.6304>
       <-4.2931 0.314 18.0722>
       <-4.1076 -0.3494 17.6517>
       <-3.0537 -0.4029 15.3994>
       <-25.1743 0.2114 10.6203>
       <-25.4219 0.5735 10.1406>
       <-25.949 1.4296 10.1914>
       <-28.5301 0.0798 11.5289>
       <-21.632 0.7752 18.0968>
       <-21.0531 -1.329 19.8374>
       <-21.0837 0.6879 20.35>
       <-23.9942 0.1579 22.6534>
       <-21.5013 -1.4047 20.9468>
       <-28.1174 0.7813 25.4761>
       <-13.8863 -0.0888 3.8213>
       <-13.6855 0.361 9.2411>
       <-13.962 0.4507 7.3538>
       <-14.0574 0.3652 3.2991>
       <-2.1337 -1.6121 -9.4149>
       <-6.8823 -2.322 -9.4056>
       <-30.4637 -1.9239 -8.4043>
       <-24.601 -2.2476 -9.3827>
       <-24.5301 -1.5518 -9.3583>
       <-26.1227 -1.9732 -8.6236>
       <-27.2431 -2.4095 -11.0324>
       <-23.7204 -2.0459 -12.1291>
       <-2.4196 2.3185 -9.7232>
       <-9.0131 1.8895 12.063>
       <-10.2861 -0.0838 12.2332>
       <-10.8791 -1.0287 12.3059>
       <-13.306 -0.9374 12.588>
       <-14.9216 0.5499 12.678>
       <-17.7918 1.5902 12.6344>
       <-18.2155 -0.0101 12.6108>
       <-20.4134 -0.463 12.3842>
       <-21.6427 0.7865 12.2335>
       <-24.4721 -1.2331 11.8509>
       <-25.6797 -1.8229 11.6961>
       <-6.637 -1.7806 10.6791>
       <-4.5884 -1.7899 15.177>
       <-6.6427 -1.7911 18.8011>
       <-5.6777 -1.7906 19.0996>
       <-7.4994 -1.7923 17.8623>
       <-10.4317 -1.774 18.8133>
       <-8.5424 -1.8001 21.6378>
       <-5.7982 -1.8001 22.2072>
       <-3.0026 -1.8001 24.4802>
       <-1.68 -1.8001 25.1566>
       <-0.4145 -1.8001 21.6984>
       <0.1973 -1.8001 19.7308>
       <0.9052 -1.8001 19.0462>
       <-1.5107 -1.7135 19.7145>
       <-2.4893 -1.8001 15.4716>
       <-3.4116 -1.8001 16.9983>
       <-5.4128 2.1362 16.1987>
       <-3.8968 2.1351 15.6428>
       <-8.6136 2.1408 17.8923>
       <-7.3132 2.1463 20.5365>
       <-6.4979 2.1463 23.1739>
       <-4.1659 2.1463 20.8126>
       <-1.927 2.1463 23.851>
       <0.5504 2.1463 20.1726>
       <-1.6016 2.1463 17.4625>
       <-3.7295 2.0897 16.983>
       <-27.0147 -1.7915 11.9298>
       <-27.6568 -1.7895 15.2122>
       <-24.4201 -1.7931 19.0553>
       <-22.2231 -1.7972 18.2631>
       <-25.4327 -1.8 19.4115>
       <-23.4308 -1.8001 20.1484>
       <-25.3703 -1.8001 21.1892>
       <-25.3042 -1.8001 20.8637>
       <-30.5482 -1.8001 24.5659>
       <-30.6658 -1.8001 17.899>
       <-28.9986 -1.8001 19.1211>
       <-29.4179 2.1353 13.2257>
       <-24.5649 2.1388 18.7798>
       <-25.7948 2.1367 19.1045>
       <-25.7094 2.1463 19.6824>
       <-26.5438 2.1463 22.7153>
       <-24.489 2.1463 21.1773>
       <-30.9769 2.1463 18.4065>
       <-4.7863 -2.8553 -9.9216>
       <-5.8277 -2.8865 -9.7762>
       <-28.2636 -2.6198 -8.473>
       <-26.0542 -3.0072 -10.2618>
       <-24.9617 -3.0293 -10.5717>
       <-25.6282 -2.7954 -11.562>
       <-3.2729 3.2231 -8.3903>
       <-2.7008 2.9075 -7.4165>
       <-2.0584 2.4661 -6.61>
       <-2.061 3.0314 -7.9669>
       <-5.129 2.9005 -8.306>
       <-6.4961 3.2052 -9.3462>
       <-27.8818 3.2183 -10.0454>
       <-25.9843 3.5089 -9.9094>
       <-16.746 3.1992 17.3396>
       <-16.8817 -2.8468 15.4094>
       <-17.5065 -2.8491 16.5872>
       <-21.7103 -2.8637 -14.0338>
       <-16.6819 -2.7859 -15.0279>
       <-27.2515 -2.097 -9.5393>
       <-27.1318 -1.9696 -7.9389>
       <-11.4037 -2.861 -14.0063>
       <-12.7159 -2.862 -14.016>
       <-7.3678 -2.7729 -13.1392>
       <-4.6165 -1.9974 -9.3027>
       <-7.3858 -2.2033 -9.7509>
       <-6.4112 -2.3621 -7.5466>
       <-10.3777 -2.4212 -8.2836>
       <-12.7194 -2.3905 -8.7689>
       <-15.9149 -2.1774 -8.7166>
       <-10.1802 -2.3437 -9.3024>
       <-11.056 -2.3426 -11.4972>
       <-12.6812 -2.3045 -11.0389>
       <-15.8271 -2.456 -12.8506>
       <-15.5933 -2.7725 -13.6337>
       <-19.2661 -2.2589 -10.4952>
       <-19.3715 -2.4112 -12.3189>
       <-21.2147 -2.5444 -13.2359>
       <-25.1696 -2.3125 -11.1258>
       <-22.0273 -2.1711 -9.4097>
       <-18.1715 -2.5071 -9.3346>
       <-22.369 -2.4495 -8.6462>
       <-11.2258 2.7725 -6.4962>
       <-8.749 2.6318 -5.5008>
       <-9.4228 2.8757 -6.5569>
       <-8.5642 2.9809 -7.9441>
       <-5.5109 2.6847 -8.3725>
       <-7.348 2.7795 -9.6034>
       <-10.7546 3.0538 -12.9705>
       <-7.7926 2.7569 -9.3301>
       <-4.65 3.3148 -12.6762>
       <-12.2885 3.3801 -14.9137>
       <-13.6645 3.3909 -14.9602>
       <-19.2712 3.5261 -14.7468>
       <-15.3504 3.0531 -12.9568>
       <-12.6914 2.859 -10.6381>
       <-18.4927 2.8161 -10.1281>
       <-26.9093 3.3421 -12.9482>
       <-19.6794 2.8646 -10.7046>
       <-24.5279 2.8238 -10.1552>
       <-21.2495 2.9778 -12.0544>
       <-26.1155 2.7964 -9.8302>
       <-19.5002 3.1016 -9.4572>
       <-18.3752 3.104 -9.4387>
       <-26.6626 2.9099 -7.0324>
       <-23.7539 -2.3919 15.6053>
       <-22.2152 -2.3919 15.5977>
       <-20.5986 -2.3899 14.6397>
       <-19.6985 -2.3896 14.8878>
       <-16.7319 -2.3911 13.224>
       <-18.6194 -2.3919 17.7808>
       <-16.2433 -2.3919 18.2296>
       <-14.1534 -2.3919 16.6935>
       <-14.8994 -2.3919 15.9796>
       <-13.3854 -2.39 14.8116>
       <-12.7181 -2.3908 16.6002>
       <-8.4624 -2.3919 13.1277>
       <-7.4123 -2.2071 11.8876>
       <-6.5959 2.7381 13.5746>
       <-6.9178 2.7259 14.9873>
       <-10.9266 2.7366 15.3864>
       <-10.7521 2.7377 17.0944>
       <-14.9429 2.7376 13.982>
       <-15.9164 2.7381 15.7468>
       <-14.3099 2.7381 18.6423>
       <-17.6503 2.6026 18.9178>
       <-17.8624 2.7381 18.2147>
       <-16.9616 2.7381 16.6973>
       <-18.9028 2.7352 13.9375>
       <-19.3004 2.7367 16.0603>
       <-21.9606 2.738 13.2142>
       <-24.0428 2.7332 13.8015>
       <-23.748 2.7338 15.2265>
       <-24.343 2.7381 17.0151>
      ]
     }}) {isSetParameters=true}]
  emitterUpdateScript [
   &EmitterState()
   &SpawnRate(Spawn Probability=0.3, SpawnRate=15, Use Spawn Probability=true)
  ]
  particleSpawnScript [
   &InitializeParticle(
    Color=$System.Sparkle Color
    Lifetime Max=1
    Lifetime Min=0.75
    Lifetime Mode=::NewEnumerator1
    Sprite Rotation Mode=::NewEnumerator2
    Sprite Size Mode=::NewEnumerator3
    Uniform Sprite Size=&SelectFloatFromArray(
     Array Sampling Mode=::NewEnumerator0
     Direct Array Index=?
     Float Selection Array=DataInterface() {propertyValues {FloatData [5,6,10,15]}}
     Random Seed=0
    )
    UsePositionOffset=false
    Write Material Random=true
   )
   /SetVariables(Particles.RandomBone:NiagaraInt32=&RandomRangeInt(Minimum=0, Maximum=17)) {isSetParameters=true}
   /SetVariables(Particles.Position=&SelectVectorFromArray(Array Sampling Mode=::NewEnumerator0, Direct Array Index=?, Vector Selection Array=$Emitter.Position Array, Random Seed=0)) {isSetParameters=true}
   &CameraOffset(Camera Offset Amount=3)
  ]
  particleUpdateScript [
   &ParticleState()
   &ScaleSpriteSize(Uniform Curve Sprite Scale=DataInterface() {propertyValues {Curve {keys [
        {interpMode=RCIM_Cubic, tangentMode=RCTM_User, time=0, value=0}
        {interpMode=RCIM_Cubic, tangentMode=RCTM_User, time=0.5, value=1}
        {tangentMode=RCTM_User, time=1, value=0}
       ]}}})
   &ScaleColor(
    Scale Alpha=&FNTimeOfDayScalar(Night Scale=0.5, Morning Scale=0.75, Evening Scale=0.75)
    Scale RGB=&VectorFromFloat(Value=&FNTimeOfDayScalar(Night Scale=0.2, Morning Scale=0.5, Evening Scale=0.5))
    ScaleA=true
    ScaleRGB=true
   )
   &SpriteRotationRate(Rotation Rate=&FloatFromCurve(
     FloatCurve=DataInterface() {propertyValues {Curve {keys [
         {interpMode=RCIM_Cubic, tangentMode=RCTM_User, time=0.2, value=0}
         {arriveTangent=3.4687, interpMode=RCIM_Cubic, leaveTangent=3.4687, tangentMode=RCTM_User, time=1, value=1}
        ]}}}
     Scale Curve=800
    ))
  ]
  rendererClasses [&NiagaraSpriteRendererProperties]
  renderers [{rendererClass=&NiagaraSpriteRendererProperties}]
 }}
summary {
 systemName=NS_Pickaxe_BananaPhilosopher_Idle
 emitters [{emitterName=Sparkles, rendererClasses [&NiagaraSpriteRendererProperties]}]
 userVariables [
  User.bIsFrontEnd {type=NiagaraBool}
  User.bIsFrontEndPreview {type=NiagaraBool}
 ]
}
